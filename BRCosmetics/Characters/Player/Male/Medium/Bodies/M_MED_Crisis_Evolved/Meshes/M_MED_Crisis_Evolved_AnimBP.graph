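# Animation Blueprint for the Crisis Evolved medium male body cosmetic, layering secondary motion and deformation fixes onto the attached parent mesh's pose.
# The copied pose runs through twist and sculptor Control Rigs, then eight crystal bones get additive drift and low-alpha spring Anim Dynamics.
# Each update refreshes three time-offset random vectors shared among the crystals, and elbow bend values that blend corrective bone offsets.
# Fixed crystal 07 offsets and a partial head_offset look-at finish the pose.
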AnimBlueprint /BRCosmetics/Characters/Player/Male/Medium/Bodies/M_MED_Crisis_Evolved/Meshes/M_MED_Crisis_Evolved_AnimBP
target_skeleton=&/M_MED_Crisis_Evolved_Skeleton
graphs [
 AnimGraph {nodes [
   n1 = Control Rig(Source=Copy Pose From Mesh() {bUseAttachedParent=true}) {
    alphaBoolBlend {blendOption=Linear}
    controlRigAssetReference {blueprintRigClass=&/M_MED_Crisis_Evolved_CtrlRig_twists.M_MED_Crisis_Evolved_CtrlRig_twists_C}
   }
   n2 = Control Rig(Source=@n1) {
    alphaBoolBlend {blendOption=Linear}
    controlRigAssetReference {blueprintRigClass=&/M_MED_Crisis_Evolved_CtrlRig_sculptors.M_MED_Crisis_Evolved_CtrlRig_sculptors_C}
   }
   n3 = Get Crystal Movement 01() -> (Crystal Movement 01:Vector)
   n4 = "Transform (Modify) Bone - Bone: dyn_crystal_01"!(ComponentPose=Local To Component(LocalPose=@n2), Translation=@n3, Rotation="0,0,0", Scale="1,1,1") -> (Pose) {
    alpha=0.3
    rotationSpace=BCS_WorldSpace
    translationMode=BMM_Additive
    translationSpace=BCS_BoneSpace
    alphaBoolBlend {blendOption=Linear}
    alphaScaleBiasClamp {interpSpeedDecreasing=2, interpSpeedIncreasing=2}
    boneToModify {boneName=dyn_crystal_01}
   }
   n5 = Anim Dynamics - Bone: dyn_crystal_01(ComponentPose=@n4, Alpha=0.05, ExternalForce="0,0,0") {
    alpha=0.05
    angularDampingOverride=1
    angularSpringConstant=20000
    bAngularSpring=true
    bGravityOverrideInSimSpace=true
    bLinearSpring=true
    bOverrideAngularDamping=true
    bOverrideLinearDamping=true
    bUseGravityOverride=true
    gravityScale=0.35
    linearDampingOverride=1
    linearSpringConstant=250
    simulationSpace=RootRelative
    alphaBoolBlend {blendOption=Linear}
    boundBone {boneName=dyn_crystal_01}
    componentAppliedLinearAccClamp=<3000 3000 3000>
    componentLinearAccScale=<0.3 0.3 1>
    componentLinearVelScale=<0.3 0.3 1>
    gravityOverride {z=-980}
    physicsBodyDefinitions [{
      sphereCollisionRadius=0
      boundBone {boneName=dyn_crystal_01}
      boxExtents=<20 30 30>
      constraintSetup {angularLimitsMax=<0 0 0>, angularLimitsMin=<0 0 0>, angularTarget=<1 0 0>, linearAxesMax=<0 50 50>, linearAxesMin=<-50 -50 -50>}
      localJointOffset=<0 0 0>
     }]
   }
   n6 = Get Crystal Movement 02() -> (Crystal Movement 02:Vector)
   n7 = "Transform (Modify) Bone - Bone: dyn_crystal_02"!(ComponentPose=@n5, Translation=@n6, Rotation="0,0,0", Scale="1,1,1") -> (Pose) {
    alpha=0.3
    rotationSpace=BCS_WorldSpace
    translationMode=BMM_Additive
    translationSpace=BCS_BoneSpace
    alphaBoolBlend {blendOption=Linear}
    alphaScaleBiasClamp {interpSpeedDecreasing=2, interpSpeedIncreasing=2}
    boneToModify {boneName=dyn_crystal_02}
   }
   n8 = Anim Dynamics - Bone: dyn_crystal_02(ComponentPose=@n7, Alpha=0.05, ExternalForce="0,0,0") {
    alpha=0.05
    angularDampingOverride=1
    angularSpringConstant=20000
    bAngularSpring=true
    bGravityOverrideInSimSpace=true
    bLinearSpring=true
    bOverrideAngularDamping=true
    bOverrideLinearDamping=true
    bUseGravityOverride=true
    gravityScale=0.35
    linearDampingOverride=1
    linearSpringConstant=185
    simulationSpace=RootRelative
    alphaBoolBlend {blendOption=Linear}
    boundBone {boneName=dyn_crystal_02}
    componentAppliedLinearAccClamp=<3000 3000 3000>
    componentLinearAccScale=<0.3 0.3 1>
    componentLinearVelScale=<0.3 0.3 1>
    gravityOverride {z=-980}
    physicsBodyDefinitions [{
      sphereCollisionRadius=0
      boundBone {boneName=dyn_crystal_02}
      boxExtents=<20 30 30>
      constraintSetup {angularLimitsMax=<0 0 0>, angularLimitsMin=<0 0 0>, angularTarget=<1 0 0>, linearAxesMax=<0 50 50>, linearAxesMin=<-50 -50 -50>}
      localJointOffset=<0 0 0>
     }]
   }
   n9 = Get Crystal Movement 03() -> (Crystal Movement 03:Vector)
   n10 = "Transform (Modify) Bone - Bone: dyn_crystal_03"!(ComponentPose=@n8, Translation=@n9, Rotation="0,0,0", Scale="1,1,1") -> (Pose) {
    alpha=0.3
    rotationSpace=BCS_WorldSpace
    translationMode=BMM_Additive
    translationSpace=BCS_BoneSpace
    alphaBoolBlend {blendOption=Linear}
    alphaScaleBiasClamp {interpSpeedDecreasing=2, interpSpeedIncreasing=2}
    boneToModify {boneName=dyn_crystal_03}
   }
   n11 = Anim Dynamics - Bone: dyn_crystal_03(ComponentPose=@n10, Alpha=0.05, ExternalForce="0,0,0") {
    alpha=0.05
    angularDampingOverride=1
    angularSpringConstant=20000
    bAngularSpring=true
    bGravityOverrideInSimSpace=true
    bLinearSpring=true
    bOverrideAngularDamping=true
    bOverrideLinearDamping=true
    bUseGravityOverride=true
    gravityScale=0.35
    linearDampingOverride=1
    linearSpringConstant=225
    simulationSpace=RootRelative
    alphaBoolBlend {blendOption=Linear}
    boundBone {boneName=dyn_crystal_03}
    componentAppliedLinearAccClamp=<3000 3000 3000>
    componentLinearAccScale=<0.3 0.3 1>
    componentLinearVelScale=<0.3 0.3 1>
    gravityOverride {z=-980}
    physicsBodyDefinitions [{
      sphereCollisionRadius=0
      boundBone {boneName=dyn_crystal_03}
      boxExtents=<20 30 30>
      constraintSetup {angularLimitsMax=<0 0 0>, angularLimitsMin=<0 0 0>, angularTarget=<1 0 0>, linearAxesMax=<0 50 50>, linearAxesMin=<-50 -50 -50>}
      localJointOffset=<0 0 0>
     }]
   }
   n12 = "Transform (Modify) Bone - Bone: dyn_crystal_04"!(ComponentPose=@n11, Translation=@n3, Rotation="0,0,0", Scale="1,1,1") -> (Pose) {
    alpha=0.3
    rotationSpace=BCS_WorldSpace
    translationMode=BMM_Additive
    translationSpace=BCS_BoneSpace
    alphaBoolBlend {blendOption=Linear}
    alphaScaleBiasClamp {interpSpeedDecreasing=2, interpSpeedIncreasing=2}
    boneToModify {boneName=dyn_crystal_04}
   }
   n13 = Anim Dynamics - Bone: dyn_crystal_04(ComponentPose=@n12, Alpha=0.05, ExternalForce="0,0,0") {
    alpha=0.05
    angularDampingOverride=1
    angularSpringConstant=20000
    bAngularSpring=true
    bGravityOverrideInSimSpace=true
    bLinearSpring=true
    bOverrideAngularDamping=true
    bOverrideLinearDamping=true
    bUseGravityOverride=true
    gravityScale=0.35
    linearDampingOverride=1
    linearSpringConstant=280
    simulationSpace=RootRelative
    alphaBoolBlend {blendOption=Linear}
    boundBone {boneName=dyn_crystal_04}
    componentAppliedLinearAccClamp=<3000 3000 3000>
    componentLinearAccScale=<0.3 0.3 1>
    componentLinearVelScale=<0.3 0.3 1>
    gravityOverride {z=-980}
    physicsBodyDefinitions [{
      sphereCollisionRadius=0
      boundBone {boneName=dyn_crystal_04}
      boxExtents=<20 30 30>
      constraintSetup {angularLimitsMax=<0 0 0>, angularLimitsMin=<0 0 0>, angularTarget=<1 0 0>, linearAxesMax=<0 50 50>, linearAxesMin=<-50 -50 -50>}
      localJointOffset=<0 0 0>
     }]
   }
   n14 = "Transform (Modify) Bone - Bone: dyn_crystal_05"!(ComponentPose=@n13, Translation=@n9, Rotation="0,0,0", Scale="1,1,1") -> (Pose) {
    alpha=0.3
    rotationSpace=BCS_WorldSpace
    translationMode=BMM_Additive
    translationSpace=BCS_BoneSpace
    alphaBoolBlend {blendOption=Linear}
    alphaScaleBiasClamp {interpSpeedDecreasing=2, interpSpeedIncreasing=2}
    boneToModify {boneName=dyn_crystal_05}
   }
   n15 = Anim Dynamics - Bone: dyn_crystal_05(ComponentPose=@n14, Alpha=0.05, ExternalForce="0,0,0") {
    alpha=0.05
    angularDampingOverride=1
    angularSpringConstant=20000
    bAngularSpring=true
    bGravityOverrideInSimSpace=true
    bLinearSpring=true
    bOverrideAngularDamping=true
    bOverrideLinearDamping=true
    bUseGravityOverride=true
    gravityScale=0.35
    linearDampingOverride=1
    linearSpringConstant=200
    simulationSpace=RootRelative
    alphaBoolBlend {blendOption=Linear}
    boundBone {boneName=dyn_crystal_05}
    componentAppliedLinearAccClamp=<3000 3000 3000>
    componentLinearAccScale=<0.2 0.2 1>
    componentLinearVelScale=<0.15 0.15 1>
    gravityOverride {z=-980}
    physicsBodyDefinitions [{
      sphereCollisionRadius=0
      boundBone {boneName=dyn_crystal_05}
      boxExtents=<20 30 30>
      constraintSetup {angularLimitsMax=<0 0 0>, angularLimitsMin=<0 0 0>, angularTarget=<1 0 0>, linearAxesMax=<0 50 50>, linearAxesMin=<-50 -50 -50>}
      localJointOffset=<0 0 0>
     }]
   }
   n16 = "Transform (Modify) Bone - Bone: dyn_crystal_06"!(ComponentPose=@n15, Translation=@n6, Rotation="0,0,0", Scale="1,1,1") -> (Pose) {
    alpha=0.3
    rotationSpace=BCS_WorldSpace
    translationMode=BMM_Additive
    translationSpace=BCS_BoneSpace
    alphaBoolBlend {blendOption=Linear}
    alphaScaleBiasClamp {interpSpeedDecreasing=2, interpSpeedIncreasing=2}
    boneToModify {boneName=dyn_crystal_06}
   }
   n17 = Anim Dynamics - Bone: dyn_crystal_06(ComponentPose=@n16, Alpha=0.05, ExternalForce="0,0,0") {
    alpha=0.05
    angularDampingOverride=1
    angularSpringConstant=20000
    bAngularSpring=true
    bGravityOverrideInSimSpace=true
    bLinearSpring=true
    bOverrideAngularDamping=true
    bOverrideLinearDamping=true
    bUseGravityOverride=true
    gravityScale=0.35
    linearDampingOverride=1
    linearSpringConstant=250
    simulationSpace=RootRelative
    alphaBoolBlend {blendOption=Linear}
    boundBone {boneName=dyn_crystal_06}
    componentAppliedLinearAccClamp=<3000 3000 3000>
    componentLinearAccScale=<0.3 0.3 1>
    componentLinearVelScale=<0.3 0.3 1>
    gravityOverride {z=-980}
    physicsBodyDefinitions [{
      sphereCollisionRadius=0
      boundBone {boneName=dyn_crystal_06}
      boxExtents=<20 30 30>
      constraintSetup {angularLimitsMax=<0 0 0>, angularLimitsMin=<0 0 0>, angularTarget=<1 0 0>, linearAxesMax=<0 50 50>, linearAxesMin=<-50 -50 -50>}
      localJointOffset=<0 0 0>
     }]
   }
   n18 = "Transform (Modify) Bone - Bone: dyn_crystal_07"!(ComponentPose=@n17, Translation=@n6, Rotation="0,0,0", Scale="1,1,1") -> (Pose) {
    alpha=0.3
    rotationSpace=BCS_WorldSpace
    translationMode=BMM_Additive
    translationSpace=BCS_BoneSpace
    alphaBoolBlend {blendOption=Linear}
    alphaScaleBiasClamp {interpSpeedDecreasing=2, interpSpeedIncreasing=2}
    boneToModify {boneName=dyn_crystal_07}
   }
   n19 = Anim Dynamics - Bone: dyn_crystal_07(ComponentPose=@n18, Alpha=0.05, ExternalForce="0,0,0") {
    alpha=0.05
    angularDampingOverride=1
    angularSpringConstant=20000
    bAngularSpring=true
    bGravityOverrideInSimSpace=true
    bLinearSpring=true
    bOverrideAngularDamping=true
    bOverrideLinearDamping=true
    bUseGravityOverride=true
    gravityScale=0.35
    linearDampingOverride=1
    linearSpringConstant=180
    simulationSpace=RootRelative
    alphaBoolBlend {blendOption=Linear}
    boundBone {boneName=dyn_crystal_07}
    componentAppliedLinearAccClamp=<3000 3000 3000>
    componentLinearAccScale=<0.1 0.1 1>
    componentLinearVelScale=<0.2 0.2 1>
    gravityOverride {z=-980}
    physicsBodyDefinitions [{
      sphereCollisionRadius=0
      boundBone {boneName=dyn_crystal_07}
      boxExtents=<20 30 30>
      constraintSetup {angularLimitsMax=<0 0 0>, angularLimitsMin=<0 0 0>, angularTarget=<1 0 0>, linearAxesMax=<0 50 50>, linearAxesMin=<-50 -50 -50>}
      localJointOffset=<0 0 0>
     }]
   }
   n20 = "Transform (Modify) Bone - Bone: dyn_crystal_08"!(ComponentPose=@n19, Translation=@n9, Rotation="0,0,0", Scale="1,1,1") -> (Pose) {
    alpha=0.3
    rotationSpace=BCS_WorldSpace
    translationMode=BMM_Additive
    translationSpace=BCS_BoneSpace
    alphaBoolBlend {blendOption=Linear}
    alphaScaleBiasClamp {interpSpeedDecreasing=2, interpSpeedIncreasing=2}
    boneToModify {boneName=dyn_crystal_08}
   }
   n21 = Anim Dynamics - Bone: dyn_crystal_08(ComponentPose=@n20, Alpha=0.05, ExternalForce="0,0,0") {
    alpha=0.05
    angularDampingOverride=1
    angularSpringConstant=20000
    bAngularSpring=true
    bGravityOverrideInSimSpace=true
    bLinearSpring=true
    bOverrideAngularDamping=true
    bOverrideLinearDamping=true
    bUseGravityOverride=true
    gravityScale=0.35
    linearDampingOverride=1
    linearSpringConstant=200
    simulationSpace=RootRelative
    alphaBoolBlend {blendOption=Linear}
    boundBone {boneName=dyn_crystal_08}
    componentAppliedLinearAccClamp=<3000 3000 3000>
    componentLinearAccScale=<0.3 0.3 1>
    componentLinearVelScale=<0.3 0.3 1>
    gravityOverride {z=-980}
    physicsBodyDefinitions [{
      sphereCollisionRadius=0
      boundBone {boneName=dyn_crystal_08}
      boxExtents=<20 30 30>
      constraintSetup {angularLimitsMax=<0 0 0>, angularLimitsMin=<0 0 0>, angularTarget=<1 0 0>, linearAxesMax=<0 50 50>, linearAxesMin=<-50 -50 -50>}
      localJointOffset=<0 0 0>
     }]
   }
   n22 = "Transform (Modify) Bone - Bone: dyn_crystal_07"!(ComponentPose=@n21, Alpha:float=1, Translation="0,0,-6", Rotation="0,0,0", Scale="1,1,1") -> (Pose) {
    alpha=0.3
    rotationSpace=BCS_WorldSpace
    translationMode=BMM_Additive
    translationSpace=BCS_BoneSpace
    alphaBoolBlend {blendOption=Linear}
    alphaScaleBiasClamp {bInterpResult=true, bMapRange=true, interpSpeedDecreasing=2, interpSpeedIncreasing=2}
    boneToModify {boneName=dyn_crystal_07}
    translation {z=-6}
   }
   n23 = "Transform (Modify) Bone - Bone: dyn_crystal_07"(ComponentPose=@n22, Alpha=1, Translation="0,0,-5", Rotation="0,0,0", Scale="1,1,1") {
    alpha=0.3
    rotationSpace=BCS_WorldSpace
    translationMode=BMM_Additive
    translationSpace=BCS_BoneSpace
    alphaBoolBlend {blendOption=Linear}
    alphaScaleBiasClamp {bInterpResult=true, bMapRange=true, interpSpeedDecreasing=2, interpSpeedIncreasing=2}
    boneToModify {boneName=dyn_crystal_07}
    translation {z=-5}
   }
   n24 = "Transform (Modify) Bone - Bone: deform_elbow_fr_l"(ComponentPose=@n23, Alpha=Get Elbow Rot L() -> (Elbow Rot L:double), Translation="10,15,0", Rotation="0,0,0", Scale="1,1,1") {
    rotationMode=BMM_Additive
    rotationSpace=BCS_BoneSpace
    scaleMode=BMM_Additive
    scaleSpace=BCS_BoneSpace
    translationMode=BMM_Additive
    translationSpace=BCS_BoneSpace
    alphaBoolBlend {blendOption=Linear}
    boneToModify {boneName=deform_elbow_fr_l}
    translation=<10 15>
   }
   n25 = "Transform (Modify) Bone - Bone: deform_elbow_fr_r"(ComponentPose=@n24, Alpha=Get Elbow Rot R() -> (Elbow Rot R:double), Translation="10,15,0", Rotation="0,0,0", Scale="1,1,1") {
    rotationMode=BMM_Additive
    rotationSpace=BCS_BoneSpace
    scaleMode=BMM_Additive
    scaleSpace=BCS_BoneSpace
    translationMode=BMM_Additive
    translationSpace=BCS_BoneSpace
    alphaBoolBlend {blendOption=Linear}
    boneToModify {boneName=deform_elbow_fr_r}
    translation=<10 15>
   }
   n26 = Look At - Bone: head_offset(ComponentPose=@n25, Alpha=0.4) {
    alphaBoolBlend {blendOption=Linear}
    boneToModify {boneName=head_offset}
    lookAtLocation {x=0}
    lookAtTarget {boneReference {boneName=head_offset_AIM}}
    lookAt_Axis {axis=<1 0>}
   }
   Output Pose(Result=Component To Local(ComponentPose=@n26)) {layerGroup=None}
  ]}
 EventGraph {nodes [
   Event BlueprintUpdateAnimation()
   > n1 = Sequence!(execute) -> (then_0, then_1, then_2:Exec)
   n2 = GetOwnerSkeletalMeshComponent()
   n3 = Random Vector Function!(
    execute=@n1.then_2
    CharacterPartAnimInstance:Custom Character Part Anim Instance Object Reference
    SkelMeshComponent:Mesh Component Object Reference=@n2
    JointName:Name=None
    GameTimeInSeconds:double=GetGameTimeInSeconds()
    TimeOffset:double=0
    PerlinScale:double=0
    Frequency:Vector="0.9,1.1,0.75"
    Amplitude:Vector="0.3,0.7,0.5"
    Offset:Vector="0,0,0"
   ) -> (then, VectorOut:Vector)
   > Set Crystal Movement 01(Crystal Movement 01:Vector=@n3) -> (Output_Get:Vector)
   > n4 = Random Vector Function(
    SkelMeshComponent=@n2
    JointName=None
    GameTimeInSeconds=GetGameTimeInSeconds()
    TimeOffset=3
    PerlinScale=0
    Frequency="0.9,1.1,0.75"
    Amplitude="0.3,0.7,0.5"
    Offset="0,0,0"
   )
   > Set Crystal Movement 02(Crystal Movement 02:Vector=@n4) -> (Output_Get:Vector)
   > n5 = Random Vector Function(
    SkelMeshComponent=@n2
    JointName=None
    GameTimeInSeconds=GetGameTimeInSeconds()
    TimeOffset=-2
    PerlinScale=0
    Frequency="0.9,1.1,0.75"
    Amplitude="0.3,0.7,0.5"
    Offset="0,0,0"
   )
   > Set Crystal Movement 03(Crystal Movement 03:Vector=@n5) -> (Output_Get:Vector)
   n6 = BreakTransform(InTransform=GetDeltaTransformFromRefPose(self=GetOwnerSkeletalMeshComponent(), BoneName=lowerarm_l, BaseName=upperarm_l))
   Set Elbow Rot L(execute=@n1.then_0, Elbow Rot L:double=MapRangeClamped(Value=BreakRotator(InRot=@n6.Rotation).Yaw, InRangeA=0, InRangeB=110)) -> (Output_Get:double)
   n7 = BreakTransform(InTransform=GetDeltaTransformFromRefPose(self=GetOwnerSkeletalMeshComponent(), BoneName=lowerarm_r, BaseName=upperarm_r))
   Set Elbow Rot R(execute=@n1.then_1, Elbow Rot R:double=MapRangeClamped(Value=BreakRotator(InRot=@n7.Rotation).Yaw, InRangeA=0, InRangeB=110)) -> (Output_Get:double)
  ]}
]
variables [
 Crystal Movement 01 {type=Vector, default=<-0.033 -0.1489 0.2092>}
 Crystal Movement 02 {type=Vector, default=<-0.0094 0.2439 -0.2712>}
 Crystal Movement 03 {type=Vector, default=<0.2849 -0.56 0.3259>}
 Elbow Rot L {type=number}
 Elbow Rot R {type=number}
]
