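# Control Rig for the Crisis Evolved male body's face accessory, adjusting crown and crystal bones procedurally as the head rotates.
# It remaps the Z Euler angle of head relative to spine_05 into a clamped, inverted value that rotates dyn_head_MSTR, parent of the crown bones.
# The same value weights two stacked Y-translation offsets on dyn_crystal_r_04, the second starting at a higher threshold.
# No other crystal bone is driven.
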ControlRigBlueprint /BRCosmetics/Characters/Player/Male/Medium/Bodies/M_MED_Crisis_Evolved/Meshes/Parts/M_MED_Crisis_Evolved_FaceAcc_CtrlRig
graphs [
 RigVMFunctionLibrary {links [], nodes []}
 RigVMModel {nodes [
   n1 = &RigUnit_BeginExecution()
   n2 = &RigVMFunction_MathFloatRemap(
    SourceMaximum=17
    SourceMinimum=-40
    TargetMaximum=0
    TargetMinimum=40
    bClamp=true
    Value=&RigVMFunction_MathQuaternionToEuler(
     RotationOrder=ZYX
     Value=&RigUnit_GetRelativeTransformForItem(Child=(Type=Bone,Name="head"), Parent=(Type=Bone,Name="spine_05")).RelativeTransform.Rotation
    ).Result.Z
   )
   &RigUnit_ModifyTransforms(
    ItemToModify=((Item=(Type=Bone,Name="dyn_head_MSTR"),Transform=(Rotation=(X=0,Y=0,Z=0,W=1),Translation=(X=0,Y=0,Z=0),Scale3D=(X=1,Y=1,Z=1))))
    Weight=1
    ItemToModify.0.Transform.Rotation=&RigVMFunction_MathQuaternionFromEuler(Euler=(X=0,Y=0,Z=0), Euler.Z=@n2.Result).Result
    ExecutePin=@n1.ExecutePin
   )
   > &RigUnit_ModifyTransforms(
    ItemToModify=((Item=(Type=Bone,Name="dyn_crystal_r_04"),Transform=(Rotation=(X=0,Y=0,Z=0,W=1),Translation=(X=0,Y=7,Z=0),Scale3D=(X=1,Y=1,Z=1))))
    Weight=&RigVMFunction_MathFloatRemap(SourceMaximum=60, SourceMinimum=0, TargetMinimum=0, bClamp=true, Value=@n2.Result).Result
   )
   > &RigUnit_ModifyTransforms(
    ItemToModify=((Item=(Type=Bone,Name="dyn_crystal_r_04"),Transform=(Rotation=(X=0,Y=0,Z=0,W=1),Translation=(X=0,Y=4,Z=0),Scale3D=(X=1,Y=1,Z=1))))
    Weight=&RigVMFunction_MathFloatRemap(SourceMaximum=60, SourceMinimum=15, TargetMinimum=0, bClamp=true, Value=@n2.Result).Result
   )
  ]}
]
hierarchy [
 root {loc=<0 0 0>, rot=<0 0 0 1>}
 pelvis {parent=root, loc=<0 -0.7071 93.1191>, rot=<0.0127 0.707 -0.0127 -0.707>}
 spine_01 {parent=pelvis, loc=<6.7494 -0.6388 0>, rot=<0 0 0.034 -0.9994>}
 spine_02 {parent=spine_01, loc=<7.2518 -0.035 0>, rot=<0 0 0.0238 -0.9997>}
 spine_03 {parent=spine_02, loc=<6.0104 0 0>, rot=<0 0 0.0149 -0.9999>}
 spine_04 {parent=spine_03, loc=<8.1024 0 0>, rot=<0 0 -0.0132 -0.9999>}
 spine_05 {parent=spine_04, loc=<10.0146 0 0>, rot=<0 0 -0.0015 -1>}
 dyn_crystal_r_01 {parent=spine_05, loc=<37.7731 -2.8589 20.6863>, rot=<0.0284 -0.7065 -0.0284 -0.7065>}
 dyn_crystal_r_02 {parent=spine_05, loc=<36.1112 -3.9506 24.0721>, rot=<0.0284 -0.7065 -0.0284 -0.7065>}
 dyn_crystal_r_03 {parent=spine_05, loc=<33.7504 -3.2018 23.3815>, rot=<0.0284 -0.7065 -0.0284 -0.7065>}
 dyn_crystal_r_04 {parent=spine_05, loc=<29.9455 3.0088 18.0571>, rot=<0.0284 -0.7065 -0.0284 -0.7065>}
 dyn_crystal_r_05 {parent=spine_05, loc=<24.4832 2.9768 20.2866>, rot=<0.0284 -0.7065 -0.0284 -0.7065>}
 dyn_crystal_l_01 {parent=spine_05, loc=<42.5976 -3.2132 -16.1867>, rot=<0.0284 -0.7065 -0.0284 -0.7065>}
 dyn_crystal_l_02 {parent=spine_05, loc=<45.6378 -7.9413 -20.3536>, rot=<0.0284 -0.7065 -0.0284 -0.7065>}
 dyn_crystal_l_03 {parent=spine_05, loc=<38.8114 -6.2959 -21.5311>, rot=<0.0284 -0.7065 -0.0284 -0.7065>}
 dyn_crystal_l_04 {parent=spine_05, loc=<30.9803 1.0153 -27.4195>, rot=<0.0284 -0.7065 -0.0284 -0.7065>}
 dyn_crystal_l_05 {parent=spine_05, loc=<28.633 2.0373 -25.7716>, rot=<0.0284 -0.7065 -0.0284 -0.7065>}
 clavicle_l {parent=spine_05, loc=<4.3575 -0.1466 -2.3223>, rot=<0.0465 -0.7289 -0.009 -0.6829>}
 upperarm_l {parent=clavicle_l, loc=<15.5111 0 0>, rot=<0.0305 -0.536 -0.0057 -0.8437>}
 lowerarm_l {parent=upperarm_l, loc=<26.6514 0 0>, rot=<0.0391 -0.0099 -0.2453 -0.9686>}
 lowerarm_twist_02_l {parent=lowerarm_l, loc=<9.6199 0 0>, rot=<0 0 0 1>}
 upperarm_twist_01_l {parent=upperarm_l, loc=<0.0041 0 0>, rot=<0 0 0 -1>}
 upperarm_twist_02_l {parent=upperarm_l, loc=<13.7941 0 0>, rot=<0 0 0 -1>}
 clavicle_r {parent=spine_05, loc=<4.3575 -0.1466 2.3223>, rot=<0.7289 0.0465 -0.6829 0.009>}
 upperarm_r {parent=clavicle_r, loc=<-15.5111 0 0>, rot=<0.0305 -0.536 -0.0057 -0.8437>}
 lowerarm_r {parent=upperarm_r, loc=<-26.6514 0 0>, rot=<0.0391 -0.0099 -0.2453 -0.9686>}
 lowerarm_twist_02_r {parent=lowerarm_r, loc=<-9.6199 0 0>, rot=<0 0 0 1>}
 upperarm_twist_01_r {parent=upperarm_r, loc=<-0.0041 0 0>, rot=<0 0 0 1>}
 upperarm_twist_02_r {parent=upperarm_r, loc=<-13.7942 0 0>, rot=<0 0 0 1>}
 neck_01 {parent=spine_05, loc=<11.3138 0.1529 0>, rot=<0 0 -0.2003 -0.9797>}
 neck_02 {parent=neck_01, loc=<4.5627 0 0>, rot=<0 0 0.0504 -0.9987>}
 head {parent=neck_02, loc=<5.3914 0 0>, rot=<0 0 0.115 -0.9934>}
 dyn_head_MSTR {parent=head, loc=<0 0 0>, rot=<0 0 0 -1>}
 dyn_crown_l_u {parent=dyn_head_MSTR, loc=<19.0409 -0.8379 -5.3978>, rot=<0.0029 -0.7071 -0.0029 -0.7071>}
 dyn_crown_l_d {parent=dyn_head_MSTR, loc=<11.6031 -1.3308 -8.5625>, rot=<0.0029 -0.7071 -0.0029 -0.7071>}
 dyn_crown_t {parent=dyn_head_MSTR, loc=<21.453 -1.1376 -0.3196>, rot=<0.0029 -0.7071 -0.0029 -0.7071>}
 dyn_crown_r_u_01 {parent=dyn_head_MSTR, loc=<20.6762 -2.2571 5.9032>, rot=<0.0029 -0.7071 -0.0029 -0.7071>}
 dyn_crown_r_u_02 {parent=dyn_head_MSTR, loc=<16.8152 0.8505 6.081>, rot=<0.0029 -0.7071 -0.0029 -0.7071>}
 dyn_crown_r_d {parent=dyn_head_MSTR, loc=<10.1869 1.0232 7.7666>, rot=<0.0029 -0.7071 -0.0029 -0.7071>}
]
