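# Post-process-style AnimBlueprint for the Crisis Evolved face accessory, adding floating drift and springy secondary motion to its crown and crystal bones.
# It copies the attached parent's pose into a Control Rig, then chains additive bone translations, each followed by low-alpha Anim Dynamics.
# Each update, Random Vector Function generates three time-offset vectors shared across the bones.
# Crown bones get subtler, stiffer motion; crystals are looser, with per-bone gravity, sometimes upward.
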AnimBlueprint /BRCosmetics/Characters/Player/Male/Medium/Bodies/M_MED_Crisis_Evolved/Meshes/Parts/M_MED_Crisis_Evolved_FaceAcc_AnimBP
target_skeleton=&/M_MED_Crisis_Evolved_FaceAcc_Skeleton
graphs [
 AnimGraph {nodes [
   n1 = Control Rig(Source=Copy Pose From Mesh() {bUseAttachedParent=true}) {
    alphaBoolBlend {blendOption=Linear}
    controlRigAssetReference {blueprintRigClass=&/M_MED_Crisis_Evolved_FaceAcc_CtrlRig.M_MED_Crisis_Evolved_FaceAcc_CtrlRig_C}
   }
   n2 = Get Crystal Movement 01() -> (Crystal Movement 01:Vector)
   n3 = "Transform (Modify) Bone - Bone: dyn_crown_l_u"!(ComponentPose=Local To Component(LocalPose=@n1), Translation=@n2, Rotation="0,0,0", Scale="1,1,1") -> (Pose) {
    alpha=0.25
    rotationSpace=BCS_WorldSpace
    translationMode=BMM_Additive
    translationSpace=BCS_BoneSpace
    alphaBoolBlend {blendOption=Linear}
    alphaScaleBiasClamp {interpSpeedDecreasing=2, interpSpeedIncreasing=2}
    boneToModify {boneName=dyn_crown_l_u}
   }
   n4 = Anim Dynamics - Bone: dyn_crown_l_u(ComponentPose=@n3, Alpha=0.05, ExternalForce="0,0,0") {
    alpha=0.05
    angularDampingOverride=1
    angularSpringConstant=20000
    bAngularSpring=true
    bGravityOverrideInSimSpace=true
    bLinearSpring=true
    bOverrideAngularDamping=true
    bOverrideLinearDamping=true
    bUseGravityOverride=true
    gravityScale=0.35
    linearDampingOverride=1
    linearSpringConstant=200
    simulationSpace=RootRelative
    alphaBoolBlend {blendOption=Linear}
    boundBone {boneName=dyn_crown_l_u}
    componentAppliedLinearAccClamp=<3000 3000 3000>
    componentLinearAccScale=<0.3 0.5 1>
    componentLinearVelScale=<0.5 0.5 1>
    gravityOverride {z=-980}
    physicsBodyDefinitions [{
      sphereCollisionRadius=0
      boundBone {boneName=dyn_crown_l_u}
      boxExtents=<20 30 30>
      constraintSetup {angularLimitsMax=<0 0 0>, angularLimitsMin=<0 0 0>, angularTarget=<1 0 0>, linearAxesMax=<0 50 50>, linearAxesMin=<-50 -50 -50>}
      localJointOffset=<0 0 0>
     }]
   }
   n5 = Get Crystal Movement 02() -> (Crystal Movement 02:Vector)
   n6 = "Transform (Modify) Bone - Bone: dyn_crown_l_d"!(ComponentPose=@n4, Translation=@n5, Rotation="0,0,0", Scale="1,1,1") -> (Pose) {
    alpha=0.35
    rotationSpace=BCS_WorldSpace
    translationMode=BMM_Additive
    translationSpace=BCS_BoneSpace
    alphaBoolBlend {blendOption=Linear}
    alphaScaleBiasClamp {interpSpeedDecreasing=2, interpSpeedIncreasing=2}
    boneToModify {boneName=dyn_crown_l_d}
   }
   n7 = Anim Dynamics - Bone: dyn_crown_l_d(ComponentPose=@n6, Alpha=0.05, ExternalForce="0,0,0") {
    alpha=0.05
    angularDampingOverride=1
    angularSpringConstant=20000
    bAngularSpring=true
    bGravityOverrideInSimSpace=true
    bLinearSpring=true
    bOverrideAngularDamping=true
    bOverrideLinearDamping=true
    bUseGravityOverride=true
    gravityScale=0.35
    linearDampingOverride=1
    linearSpringConstant=200
    simulationSpace=RootRelative
    alphaBoolBlend {blendOption=Linear}
    boundBone {boneName=dyn_crown_l_d}
    componentAppliedLinearAccClamp=<3000 3000 3000>
    componentLinearAccScale=<0.3 0.5 1>
    componentLinearVelScale=<0.5 0.5 1>
    gravityOverride {z=-980}
    physicsBodyDefinitions [{
      sphereCollisionRadius=0
      boundBone {boneName=dyn_crown_l_d}
      boxExtents=<20 30 30>
      constraintSetup {angularLimitsMax=<0 0 0>, angularLimitsMin=<0 0 0>, angularTarget=<1 0 0>, linearAxesMax=<0 50 50>, linearAxesMin=<-50 -50 -50>}
      localJointOffset=<0 0 0>
     }]
   }
   n8 = Get Crystal Movement 03() -> (Crystal Movement 03:Vector)
   n9 = "Transform (Modify) Bone - Bone: dyn_crown_t"!(ComponentPose=@n7, Translation=@n8, Rotation="0,0,0", Scale="1,1,1") -> (Pose) {
    alpha=0.35
    rotationSpace=BCS_WorldSpace
    translationMode=BMM_Additive
    translationSpace=BCS_BoneSpace
    alphaBoolBlend {blendOption=Linear}
    alphaScaleBiasClamp {interpSpeedDecreasing=2, interpSpeedIncreasing=2}
    boneToModify {boneName=dyn_crown_t}
   }
   n10 = Anim Dynamics - Bone: dyn_crown_t(ComponentPose=@n9, Alpha=0.08, ExternalForce="0,0,0") {
    alpha=0.05
    angularDampingOverride=1
    angularSpringConstant=20000
    bAngularSpring=true
    bGravityOverrideInSimSpace=true
    bLinearSpring=true
    bOverrideAngularDamping=true
    bOverrideLinearDamping=true
    bUseGravityOverride=true
    gravityScale=0.35
    linearDampingOverride=1
    linearSpringConstant=200
    simulationSpace=RootRelative
    alphaBoolBlend {blendOption=Linear}
    boundBone {boneName=dyn_crown_t}
    componentAppliedLinearAccClamp=<3000 3000 3000>
    componentLinearAccScale=<0.3 0.5 1>
    componentLinearVelScale=<0.5 0.5 1>
    gravityOverride {z=-980}
    physicsBodyDefinitions [{
      sphereCollisionRadius=0
      boundBone {boneName=dyn_crown_t}
      boxExtents=<20 30 30>
      constraintSetup {angularLimitsMax=<0 0 0>, angularLimitsMin=<0 0 0>, angularTarget=<1 0 0>, linearAxesMax=<0 50 50>, linearAxesMin=<-50 -50 -50>}
      localJointOffset=<0 0 0>
     }]
   }
   n11 = "Transform (Modify) Bone - Bone: dyn_crown_r_u_01"!(ComponentPose=@n10, Translation=@n2, Rotation="0,0,0", Scale="1,1,1") -> (Pose) {
    alpha=0.2
    rotationSpace=BCS_WorldSpace
    translationMode=BMM_Additive
    translationSpace=BCS_BoneSpace
    alphaBoolBlend {blendOption=Linear}
    alphaScaleBiasClamp {interpSpeedDecreasing=2, interpSpeedIncreasing=2}
    boneToModify {boneName=dyn_crown_r_u_01}
   }
   n12 = Anim Dynamics - Bone: dyn_crown_r_u_01(ComponentPose=@n11, Alpha=0.06, ExternalForce="0,0,0") {
    alpha=0.05
    angularDampingOverride=1
    angularSpringConstant=20000
    bAngularSpring=true
    bGravityOverrideInSimSpace=true
    bLinearSpring=true
    bOverrideAngularDamping=true
    bOverrideLinearDamping=true
    bUseGravityOverride=true
    gravityScale=0.35
    linearDampingOverride=1
    linearSpringConstant=200
    simulationSpace=RootRelative
    alphaBoolBlend {blendOption=Linear}
    boundBone {boneName=dyn_crown_r_u_01}
    componentAppliedLinearAccClamp=<3000 3000 3000>
    componentLinearAccScale=<0.3 0.5 1>
    componentLinearVelScale=<0.5 0.5 1>
    gravityOverride {z=-980}
    physicsBodyDefinitions [{
      sphereCollisionRadius=0
      boundBone {boneName=dyn_crown_r_u_01}
      boxExtents=<20 30 30>
      constraintSetup {angularLimitsMax=<0 0 0>, angularLimitsMin=<0 0 0>, angularTarget=<1 0 0>, linearAxesMax=<0 50 50>, linearAxesMin=<-50 -50 -50>}
      localJointOffset=<0 0 0>
     }]
   }
   n13 = "Transform (Modify) Bone - Bone: dyn_crown_r_u_02"!(ComponentPose=@n12, Translation=@n5, Rotation="0,0,0", Scale="1,1,1") -> (Pose) {
    alpha=0.35
    rotationSpace=BCS_WorldSpace
    translationMode=BMM_Additive
    translationSpace=BCS_BoneSpace
    alphaBoolBlend {blendOption=Linear}
    alphaScaleBiasClamp {interpSpeedDecreasing=2, interpSpeedIncreasing=2}
    boneToModify {boneName=dyn_crown_r_u_02}
   }
   n14 = Anim Dynamics - Bone: dyn_crown_r_u_02(ComponentPose=@n13, Alpha=0.07, ExternalForce="0,0,0") {
    alpha=0.05
    angularDampingOverride=1
    angularSpringConstant=20000
    bAngularSpring=true
    bGravityOverrideInSimSpace=true
    bLinearSpring=true
    bOverrideAngularDamping=true
    bOverrideLinearDamping=true
    bUseGravityOverride=true
    gravityScale=0.35
    linearDampingOverride=1
    linearSpringConstant=200
    simulationSpace=RootRelative
    alphaBoolBlend {blendOption=Linear}
    boundBone {boneName=dyn_crown_r_u_02}
    componentAppliedLinearAccClamp=<3000 3000 3000>
    componentLinearAccScale=<0.3 0.5 1>
    componentLinearVelScale=<0.5 0.5 1>
    gravityOverride {z=-980}
    physicsBodyDefinitions [{
      sphereCollisionRadius=0
      boundBone {boneName=dyn_crown_r_u_02}
      boxExtents=<20 30 30>
      constraintSetup {angularLimitsMax=<0 0 0>, angularLimitsMin=<0 0 0>, angularTarget=<1 0 0>, linearAxesMax=<0 50 50>, linearAxesMin=<-50 -50 -50>}
      localJointOffset=<0 0 0>
     }]
   }
   n15 = "Transform (Modify) Bone - Bone: dyn_crown_r_d"!(ComponentPose=@n14, Translation=@n8, Rotation="0,0,0", Scale="1,1,1") -> (Pose) {
    alpha=0.35
    rotationSpace=BCS_WorldSpace
    translationMode=BMM_Additive
    translationSpace=BCS_BoneSpace
    alphaBoolBlend {blendOption=Linear}
    alphaScaleBiasClamp {interpSpeedDecreasing=2, interpSpeedIncreasing=2}
    boneToModify {boneName=dyn_crown_r_d}
   }
   n16 = Anim Dynamics - Bone: dyn_crown_r_d(ComponentPose=@n15, Alpha=0.05, ExternalForce="0,0,0") {
    alpha=0.05
    angularDampingOverride=1
    angularSpringConstant=20000
    bAngularSpring=true
    bGravityOverrideInSimSpace=true
    bLinearSpring=true
    bOverrideAngularDamping=true
    bOverrideLinearDamping=true
    bUseGravityOverride=true
    gravityScale=0.35
    linearDampingOverride=1
    linearSpringConstant=200
    simulationSpace=RootRelative
    alphaBoolBlend {blendOption=Linear}
    boundBone {boneName=dyn_crown_r_d}
    componentAppliedLinearAccClamp=<3000 3000 3000>
    componentLinearAccScale=<0.3 0.5 1>
    componentLinearVelScale=<0.5 0.5 1>
    gravityOverride {z=-980}
    physicsBodyDefinitions [{
      sphereCollisionRadius=0
      boundBone {boneName=dyn_crown_r_d}
      boxExtents=<20 30 30>
      constraintSetup {angularLimitsMax=<0 0 0>, angularLimitsMin=<0 0 0>, angularTarget=<1 0 0>, linearAxesMax=<0 50 50>, linearAxesMin=<-50 -50 -50>}
      localJointOffset=<0 0 0>
     }]
   }
   n17 = "Transform (Modify) Bone - Bone: dyn_crystal_r_01"!(ComponentPose=@n16, Translation=@n2, Rotation="0,0,0", Scale="1,1,1") -> (Pose) {
    rotationSpace=BCS_WorldSpace
    translationMode=BMM_Additive
    translationSpace=BCS_BoneSpace
    alphaBoolBlend {blendOption=Linear}
    alphaScaleBiasClamp {interpSpeedDecreasing=2, interpSpeedIncreasing=2}
    boneToModify {boneName=dyn_crystal_r_01}
   }
   n18 = Anim Dynamics - Bone: dyn_crystal_r_01(ComponentPose=@n17, Alpha=0.32, ExternalForce="0,0,0") {
    alpha=0.23
    angularDampingOverride=1
    angularSpringConstant=20000
    bAngularSpring=true
    bGravityOverrideInSimSpace=true
    bLinearSpring=true
    bOverrideAngularDamping=true
    bOverrideLinearDamping=true
    bUseGravityOverride=true
    gravityScale=0.35
    linearDampingOverride=1
    linearSpringConstant=30
    simulationSpace=RootRelative
    alphaBoolBlend {blendOption=Linear}
    boundBone {boneName=dyn_crystal_r_01}
    componentAppliedLinearAccClamp=<3000 3000 3000>
    componentLinearAccScale=<1.4 1 0.5>
    componentLinearVelScale=<0.7 0.7 0.5>
    gravityOverride {z=-100}
    physicsBodyDefinitions [{
      sphereCollisionRadius=0
      boundBone {boneName=dyn_crystal_r_01}
      boxExtents=<20 30 30>
      constraintSetup {angularLimitsMax=<0 0 0>, angularLimitsMin=<0 0 0>, angularTarget=<1 0 0>, linearAxesMax=<100 100 100>, linearAxesMin=<-100 -100 -100>}
      localJointOffset=<0 0 0>
     }]
   }
   n19 = "Transform (Modify) Bone - Bone: dyn_crystal_r_02"!(ComponentPose=@n18, Translation=@n5, Rotation="0,0,0", Scale="1,1,1") -> (Pose) {
    rotationSpace=BCS_WorldSpace
    translationMode=BMM_Additive
    translationSpace=BCS_BoneSpace
    alphaBoolBlend {blendOption=Linear}
    alphaScaleBiasClamp {interpSpeedDecreasing=2, interpSpeedIncreasing=2}
    boneToModify {boneName=dyn_crystal_r_02}
   }
   n20 = Anim Dynamics - Bone: dyn_crystal_r_02(ComponentPose=@n19, Alpha=0.23, ExternalForce="0,0,0") {
    alpha=0.2
    angularDampingOverride=1
    angularSpringConstant=20000
    bAngularSpring=true
    bGravityOverrideInSimSpace=true
    bLinearSpring=true
    bOverrideAngularDamping=true
    bOverrideLinearDamping=true
    bUseGravityOverride=true
    gravityScale=0.35
    linearDampingOverride=1
    linearSpringConstant=30
    simulationSpace=RootRelative
    alphaBoolBlend {blendOption=Linear}
    boundBone {boneName=dyn_crystal_r_02}
    componentAppliedLinearAccClamp=<3000 3000 3000>
    componentLinearAccScale=<1.2 1 0.5>
    componentLinearVelScale=<0.7 0.7 0.5>
    gravityOverride {z=-100}
    physicsBodyDefinitions [{
      sphereCollisionRadius=0
      boundBone {boneName=dyn_crystal_r_02}
      boxExtents=<20 30 30>
      constraintSetup {angularLimitsMax=<0 0 0>, angularLimitsMin=<0 0 0>, angularTarget=<1 0 0>, linearAxesMax=<100 100 100>, linearAxesMin=<-100 -100 -100>}
      localJointOffset=<0 0 0>
     }]
   }
   n21 = "Transform (Modify) Bone - Bone: dyn_crystal_r_03"!(ComponentPose=@n20, Translation=@n8, Rotation="0,0,0", Scale="1,1,1") -> (Pose) {
    rotationSpace=BCS_WorldSpace
    translationMode=BMM_Additive
    translationSpace=BCS_BoneSpace
    alphaBoolBlend {blendOption=Linear}
    alphaScaleBiasClamp {interpSpeedDecreasing=2, interpSpeedIncreasing=2}
    boneToModify {boneName=dyn_crystal_r_03}
   }
   n22 = Anim Dynamics - Bone: dyn_crystal_r_03(ComponentPose=@n21, Alpha=0.3, ExternalForce="0,0,0") {
    alpha=0.18
    angularDampingOverride=1
    angularSpringConstant=20000
    bAngularSpring=true
    bGravityOverrideInSimSpace=true
    bLinearSpring=true
    bOverrideAngularDamping=true
    bOverrideLinearDamping=true
    bUseGravityOverride=true
    gravityScale=0.35
    linearDampingOverride=1
    linearSpringConstant=30
    simulationSpace=RootRelative
    alphaBoolBlend {blendOption=Linear}
    boundBone {boneName=dyn_crystal_r_03}
    componentAppliedLinearAccClamp=<3000 3000 3000>
    componentLinearAccScale=<1.1 1 0.5>
    componentLinearVelScale=<0.7 0.7 0.5>
    gravityOverride {z=-50}
    physicsBodyDefinitions [{
      sphereCollisionRadius=0
      boundBone {boneName=dyn_crystal_r_03}
      boxExtents=<20 30 30>
      constraintSetup {angularLimitsMax=<0 0 0>, angularLimitsMin=<0 0 0>, angularTarget=<1 0 0>, linearAxesMax=<100 100 100>, linearAxesMin=<-100 -100 -100>}
      localJointOffset=<0 0 0>
     }]
   }
   n23 = "Transform (Modify) Bone - Bone: dyn_crystal_r_04"!(ComponentPose=@n22, Translation=@n2, Rotation="0,0,0", Scale="1,1,1") -> (Pose) {
    rotationSpace=BCS_WorldSpace
    translationMode=BMM_Additive
    translationSpace=BCS_BoneSpace
    alphaBoolBlend {blendOption=Linear}
    alphaScaleBiasClamp {interpSpeedDecreasing=2, interpSpeedIncreasing=2}
    boneToModify {boneName=dyn_crystal_r_04}
   }
   n24 = Anim Dynamics - Bone: dyn_crystal_r_04(ComponentPose=@n23, Alpha=0.21, ExternalForce="0,0,0") {
    alpha=0.15
    angularDampingOverride=1
    angularSpringConstant=20000
    bAngularSpring=true
    bGravityOverrideInSimSpace=true
    bLinearSpring=true
    bOverrideAngularDamping=true
    bOverrideLinearDamping=true
    bUseGravityOverride=true
    gravityScale=0.35
    linearDampingOverride=1
    linearSpringConstant=30
    simulationSpace=RootRelative
    alphaBoolBlend {blendOption=Linear}
    boundBone {boneName=dyn_crystal_r_04}
    componentAppliedLinearAccClamp=<3000 3000 3000>
    componentLinearAccScale=<1 1 0.5>
    componentLinearVelScale=<0.7 0.7 0.5>
    gravityOverride {z=250}
    physicsBodyDefinitions [{
      sphereCollisionRadius=0
      boundBone {boneName=dyn_crystal_r_04}
      boxExtents=<20 30 30>
      constraintSetup {angularLimitsMax=<0 0 0>, angularLimitsMin=<0 0 0>, angularTarget=<1 0 0>, linearAxesMax=<100 100 100>, linearAxesMin=<-100 -100 -100>}
      localJointOffset=<0 0 0>
     }]
   }
   n25 = "Transform (Modify) Bone - Bone: dyn_crystal_r_05"!(ComponentPose=@n24, Translation=@n5, Rotation="0,0,0", Scale="1,1,1") -> (Pose) {
    rotationSpace=BCS_WorldSpace
    translationMode=BMM_Additive
    translationSpace=BCS_BoneSpace
    alphaBoolBlend {blendOption=Linear}
    alphaScaleBiasClamp {interpSpeedDecreasing=2, interpSpeedIncreasing=2}
    boneToModify {boneName=dyn_crystal_r_05}
   }
   n26 = Anim Dynamics - Bone: dyn_crystal_r_05(ComponentPose=@n25, Alpha=0.1, ExternalForce="0,0,0") {
    alpha=0.13
    angularDampingOverride=1
    angularSpringConstant=20000
    bAngularSpring=true
    bGravityOverrideInSimSpace=true
    bLinearSpring=true
    bOverrideAngularDamping=true
    bOverrideLinearDamping=true
    bUseGravityOverride=true
    gravityScale=0.35
    linearDampingOverride=1
    linearSpringConstant=30
    simulationSpace=RootRelative
    alphaBoolBlend {blendOption=Linear}
    boundBone {boneName=dyn_crystal_r_05}
    componentAppliedLinearAccClamp=<3000 3000 3000>
    componentLinearAccScale=<1.2 1 0.5>
    componentLinearVelScale=<0.7 0.7 0.5>
    gravityOverride {z=-100}
    physicsBodyDefinitions [{
      sphereCollisionRadius=0
      boundBone {boneName=dyn_crystal_r_05}
      boxExtents=<20 30 30>
      constraintSetup {angularLimitsMax=<0 0 0>, angularLimitsMin=<0 0 0>, angularTarget=<1 0 0>, linearAxesMax=<100 100 100>, linearAxesMin=<-100 -100 -100>}
      localJointOffset=<0 0 0>
     }]
   }
   n27 = "Transform (Modify) Bone - Bone: dyn_crystal_l_01"!(ComponentPose=@n26, Translation=@n2, Rotation="0,0,0", Scale="1,1,1") -> (Pose) {
    rotationSpace=BCS_WorldSpace
    translationMode=BMM_Additive
    translationSpace=BCS_BoneSpace
    alphaBoolBlend {blendOption=Linear}
    alphaScaleBiasClamp {interpSpeedDecreasing=2, interpSpeedIncreasing=2}
    boneToModify {boneName=dyn_crystal_l_01}
   }
   n28 = Anim Dynamics - Bone: dyn_crystal_l_01(ComponentPose=@n27, Alpha=0.28, ExternalForce="0,0,0") {
    alpha=0.17
    angularDampingOverride=1
    angularSpringConstant=20000
    bAngularSpring=true
    bGravityOverrideInSimSpace=true
    bLinearSpring=true
    bOverrideAngularDamping=true
    bOverrideLinearDamping=true
    bUseGravityOverride=true
    gravityScale=0.35
    linearDampingOverride=1
    linearSpringConstant=30
    simulationSpace=RootRelative
    alphaBoolBlend {blendOption=Linear}
    boundBone {boneName=dyn_crystal_l_01}
    componentAppliedLinearAccClamp=<3000 3000 3000>
    componentLinearAccScale=<1 1 0.5>
    componentLinearVelScale=<0.7 0.7 0.5>
    gravityOverride {z=-100}
    physicsBodyDefinitions [{
      sphereCollisionRadius=0
      boundBone {boneName=dyn_crystal_l_01}
      boxExtents=<20 30 30>
      constraintSetup {angularLimitsMax=<0 0 0>, angularLimitsMin=<0 0 0>, angularTarget=<1 0 0>, linearAxesMax=<100 100 100>, linearAxesMin=<-100 -100 -100>}
      localJointOffset=<0 0 0>
     }]
   }
   n29 = "Transform (Modify) Bone - Bone: dyn_crystal_l_02"!(ComponentPose=@n28, Translation=@n5, Rotation="0,0,0", Scale="1,1,1") -> (Pose) {
    rotationSpace=BCS_WorldSpace
    translationMode=BMM_Additive
    translationSpace=BCS_BoneSpace
    alphaBoolBlend {blendOption=Linear}
    alphaScaleBiasClamp {interpSpeedDecreasing=2, interpSpeedIncreasing=2}
    boneToModify {boneName=dyn_crystal_l_02}
   }
   n30 = Anim Dynamics - Bone: dyn_crystal_l_02(ComponentPose=@n29, Alpha=0.36, ExternalForce="0,0,0") {
    alpha=0.21
    angularDampingOverride=1
    angularSpringConstant=20000
    bAngularSpring=true
    bGravityOverrideInSimSpace=true
    bLinearSpring=true
    bOverrideAngularDamping=true
    bOverrideLinearDamping=true
    bUseGravityOverride=true
    gravityScale=0.35
    linearDampingOverride=1
    linearSpringConstant=30
    simulationSpace=RootRelative
    alphaBoolBlend {blendOption=Linear}
    boundBone {boneName=dyn_crystal_l_02}
    componentAppliedLinearAccClamp=<3000 3000 3000>
    componentLinearAccScale=<1.4 1 0.5>
    componentLinearVelScale=<0.7 0.7 0.5>
    gravityOverride {z=-100}
    physicsBodyDefinitions [{
      sphereCollisionRadius=0
      boundBone {boneName=dyn_crystal_l_02}
      boxExtents=<20 30 30>
      constraintSetup {angularLimitsMax=<0 0 0>, angularLimitsMin=<0 0 0>, angularTarget=<1 0 0>, linearAxesMax=<100 100 100>, linearAxesMin=<-100 -100 -100>}
      localJointOffset=<0 0 0>
     }]
   }
   n31 = "Transform (Modify) Bone - Bone: dyn_crystal_l_03"!(ComponentPose=@n30, Translation=@n8, Rotation="0,0,0", Scale="1,1,1") -> (Pose) {
    rotationSpace=BCS_WorldSpace
    translationMode=BMM_Additive
    translationSpace=BCS_BoneSpace
    alphaBoolBlend {blendOption=Linear}
    alphaScaleBiasClamp {interpSpeedDecreasing=2, interpSpeedIncreasing=2}
    boneToModify {boneName=dyn_crystal_l_03}
   }
   n32 = Anim Dynamics - Bone: dyn_crystal_l_03(ComponentPose=@n31, Alpha=0.28, ExternalForce="0,0,0") {
    alpha=0.17
    angularDampingOverride=1
    angularSpringConstant=20000
    bAngularSpring=true
    bGravityOverrideInSimSpace=true
    bLinearSpring=true
    bOverrideAngularDamping=true
    bOverrideLinearDamping=true
    bUseGravityOverride=true
    gravityScale=0.35
    linearDampingOverride=1
    linearSpringConstant=30
    simulationSpace=RootRelative
    alphaBoolBlend {blendOption=Linear}
    boundBone {boneName=dyn_crystal_l_03}
    componentAppliedLinearAccClamp=<3000 3000 3000>
    componentLinearAccScale=<1.2 1 0.5>
    componentLinearVelScale=<0.7 0.7 0.5>
    gravityOverride {z=-100}
    physicsBodyDefinitions [{
      sphereCollisionRadius=0
      boundBone {boneName=dyn_crystal_l_03}
      boxExtents=<20 30 30>
      constraintSetup {angularLimitsMax=<0 0 0>, angularLimitsMin=<0 0 0>, angularTarget=<1 0 0>, linearAxesMax=<100 100 100>, linearAxesMin=<-100 -100 -100>}
      localJointOffset=<0 0 0>
     }]
   }
   n33 = "Transform (Modify) Bone - Bone: dyn_crystal_l_04"!(ComponentPose=@n32, Translation=@n2, Rotation="0,0,0", Scale="1,1,1") -> (Pose) {
    rotationSpace=BCS_WorldSpace
    translationMode=BMM_Additive
    translationSpace=BCS_BoneSpace
    alphaBoolBlend {blendOption=Linear}
    alphaScaleBiasClamp {interpSpeedDecreasing=2, interpSpeedIncreasing=2}
    boneToModify {boneName=dyn_crystal_l_04}
   }
   n34 = Anim Dynamics - Bone: dyn_crystal_l_04(ComponentPose=@n33, Alpha=0.28, ExternalForce="0,0,0") {
    alpha=0.16
    angularDampingOverride=1
    angularSpringConstant=20000
    bAngularSpring=true
    bGravityOverrideInSimSpace=true
    bLinearSpring=true
    bOverrideAngularDamping=true
    bOverrideLinearDamping=true
    bUseGravityOverride=true
    gravityScale=0.35
    linearDampingOverride=1
    linearSpringConstant=30
    simulationSpace=RootRelative
    alphaBoolBlend {blendOption=Linear}
    boundBone {boneName=dyn_crystal_l_04}
    componentAppliedLinearAccClamp=<3000 3000 3000>
    componentLinearAccScale=<1 1 0.5>
    componentLinearVelScale=<0.7 0.7 0.5>
    gravityOverride {z=100}
    physicsBodyDefinitions [{
      sphereCollisionRadius=0
      boundBone {boneName=dyn_crystal_l_04}
      boxExtents=<20 30 30>
      constraintSetup {angularLimitsMax=<0 0 0>, angularLimitsMin=<0 0 0>, angularTarget=<1 0 0>, linearAxesMax=<100 100 100>, linearAxesMin=<-100 -100 -100>}
      localJointOffset=<0 0 0>
     }]
   }
   n35 = "Transform (Modify) Bone - Bone: dyn_crystal_l_05"!(ComponentPose=@n34, Translation=@n5, Rotation="0,0,0", Scale="1,1,1") -> (Pose) {
    rotationSpace=BCS_WorldSpace
    translationMode=BMM_Additive
    translationSpace=BCS_BoneSpace
    alphaBoolBlend {blendOption=Linear}
    alphaScaleBiasClamp {interpSpeedDecreasing=2, interpSpeedIncreasing=2}
    boneToModify {boneName=dyn_crystal_l_05}
   }
   n36 = Anim Dynamics - Bone: dyn_crystal_l_05(ComponentPose=@n35, Alpha=0.23, ExternalForce="0,0,0") {
    alpha=0.14
    angularDampingOverride=1
    angularSpringConstant=20000
    bAngularSpring=true
    bGravityOverrideInSimSpace=true
    bLinearSpring=true
    bOverrideAngularDamping=true
    bOverrideLinearDamping=true
    bUseGravityOverride=true
    gravityScale=0.35
    linearDampingOverride=1
    linearSpringConstant=30
    simulationSpace=RootRelative
    alphaBoolBlend {blendOption=Linear}
    boundBone {boneName=dyn_crystal_l_05}
    componentAppliedLinearAccClamp=<3000 3000 3000>
    componentLinearAccScale=<1.5 1 0.5>
    componentLinearVelScale=<0.7 0.7 0.5>
    gravityOverride {z=-100}
    physicsBodyDefinitions [{
      sphereCollisionRadius=0
      boundBone {boneName=dyn_crystal_l_05}
      boxExtents=<20 30 30>
      constraintSetup {angularLimitsMax=<0 0 0>, angularLimitsMin=<0 0 0>, angularTarget=<1 0 0>, linearAxesMax=<100 100 100>, linearAxesMin=<-100 -100 -100>}
      localJointOffset=<0 0 0>
     }]
   }
   Output Pose(Result=Component To Local(ComponentPose=@n36)) {layerGroup=None}
  ]}
 EventGraph {nodes [
   Event BlueprintUpdateAnimation()
   n1 = GetOwnerSkeletalMeshComponent()
   > n2 = Random Vector Function!(
    execute
    CharacterPartAnimInstance:Custom Character Part Anim Instance Object Reference
    SkelMeshComponent:Mesh Component Object Reference=@n1
    JointName:Name=None
    GameTimeInSeconds:double=GetGameTimeInSeconds()
    TimeOffset:double=0
    PerlinScale:double=0
    Frequency:Vector="0.9,1.1,0.75"
    Amplitude:Vector="0.6,1.8,1.4"
    Offset:Vector="0,0,0"
   ) -> (then, VectorOut:Vector)
   > Set Crystal Movement 01(Crystal Movement 01:Vector=@n2) -> (Output_Get:Vector)
   > n3 = Random Vector Function(
    SkelMeshComponent=@n1
    JointName=None
    GameTimeInSeconds=GetGameTimeInSeconds()
    TimeOffset=3
    PerlinScale=0
    Frequency="0.9,1.1,0.75"
    Amplitude="0.6,1.8,1.4"
    Offset="0,0,0"
   )
   > Set Crystal Movement 02(Crystal Movement 02:Vector=@n3) -> (Output_Get:Vector)
   > n4 = Random Vector Function(
    SkelMeshComponent=@n1
    JointName=None
    GameTimeInSeconds=GetGameTimeInSeconds()
    TimeOffset=-2
    PerlinScale=0
    Frequency="0.9,1.1,0.75"
    Amplitude="0.6,1.8,1.4"
    Offset="0,0,0"
   )
   > Set Crystal Movement 03(Crystal Movement 03:Vector=@n4) -> (Output_Get:Vector)
  ]}
]
variables [
 Crystal Movement 01 {type=Vector}
 Crystal Movement 02 {type=Vector}
 Crystal Movement 03 {type=Vector}
]
